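# Physics asset for the Underground Rebel backpack cosmetic, giving the pack mesh simple collision and a jointed handle.
# A tilted box body covers the pack bone, and a small capsule body covers the dyn_handle bone.
# A single constraint joins the handle to the pack with both swing axes locked, so the handle rotates only about the twist axis, like a hinge.
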PhysicsAsset /BRCosmetics/Accessories/FORT_Backpacks/Backpack_M_MED_UndergroundRebel/Meshes/M_MED_UndergroundRebel_Pack_Physics
bodies {
 dyn_handle [{shapeType=Capsule, length=8.25, radius=2, shapeName=dyn_handle_capsule, center=<0 0 1.7308>, rotation=<rot -90 0.0001 -0.0003>}]
 pack [{
   shapeType=Box
   extentX=33.5
   extentY=11.5
   extentZ=25.5
   shapeName=pack_box
   center=<-0.4249 -5.5014 -1.7939>
   rotation=<rot -25 0.0001 -0.0001>
  }]
}
constraints [{
  bone1Name=dyn_handle
  bone2Name=pack
  swing1LimitDegrees=45
  swing1Motion=Locked
  swing2LimitDegrees=45
  swing2Motion=Locked
  twistLimitDegrees=90
 }]
